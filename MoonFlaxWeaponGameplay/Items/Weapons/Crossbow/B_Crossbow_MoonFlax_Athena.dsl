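; Crossbow weapon Blueprint for MoonFlax that briefly hides its pilot light meshes whenever the muzzle flash effect plays.
; The muzzle flash event calls its parent, then PilotLightToggle, which saves the scale of SM_Straight_Spline_101 and zeroes both spline meshes.
; After a one-second delay, both meshes get that single saved scale back.
; The scale is re-saved on every call, so a retrigger during the delay captures zero.
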(event Custom|PilotLightToggle
  (bind _sm_straight_spline_101 (Variables|Default|GetSM_Straight_Spline_101))
  (bind _sm_straight_spline_100 (Variables|Default|GetSM_Straight_Spline_100))
  (bind _output_get (Variables|Default|SetPilotLightScale (Variables|/Script/Engine.SceneComponent|GetRelativeScale3D _sm_straight_spline_101)))
  (Transformation|SetRelativeScale3D _sm_straight_spline_101 "0, 0, 0")
  (Transformation|SetRelativeScale3D _sm_straight_spline_100 "0, 0, 0")
  (Utilities|FlowControl|Delay 1.0)
  (Transformation|SetRelativeScale3D _sm_straight_spline_101 _output_get)
  (Transformation|SetRelativeScale3D _sm_straight_spline_100 _output_get))

(event EventMuzzleFlashFX
  (params (Persistent%20Fire :type Boolean))
  (|CallParentFunction Persistent%20Fire)
  (CallFunction|PilotLightToggle))

(fn ConstructionScript
  (|CallParentFunction))
