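; Projectile for the Moonflax explosive crossbow: a sticky bolt that attaches to pawns, buildings, vehicles or supply drops, beeps, then explodes.
; Touch, overlap and stop events attach it on authority and start a fuse chosen by target type (vehicle, pawn, other).
; Host loss either detonates it or resumes simulation; a lifetime timer, destruction or tagged overlaps also trigger the explosion.
; Cosmetic paths drive the goo cue, audio, sound indicators, explosion effects and water splash.
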(event EventBeginPlay
  ;[1 1 1] EVENT -> BeginPlay
  (Utilities|FlowControl|Sequence
    (:then_0
      (Utilities|FlowControl|Delay (Variables|Default|GetSelfStickDelay))
      (Variables|Default|SetIsSelfStickDelayOver? true))
    (:then_1
      ;;[1 1 1] With bouncy surfaces, there's a possibility for a long life. Explode after a time
      (Utilities|Time|SetTimerbyEvent (delegate Custom|Destroy) 5.0))))

(event Custom|Destroy
  ;[1 1 1] EVENT -> BeginPlay
  ;;[1 1 1] With bouncy surfaces, there's a possibility for a long life. Explode after a time
  (CallFunction|EventTriggerExplosion))

(event EventOnTouched
  (params (OtherActor :type Actor%20Object%20Reference) (OtherComp :type Primitive%20Component%20Object%20Reference) (HitResult :type Hit%20Result%20Structure%20%28by%20ref%29) (bIsOverlap :type Boolean))
  (bind (_bblockinghit _binitialoverlap _time _distance _location _impactpoint _normal _impactnormal _physmat _hitactor _hitcomponent) (Collision|BreakHitResult HitResult))
  ;[1 1 1] EVENT -> OnTouched
  (bind _output_get (Variables|Default|SetTouchedActor OtherActor))
  (if (Components|ComponentHasTag _hitcomponent "NoStick")
    (Utilities|Casting|CastToB_HidingProp _output_get)
    (else
      _))
  (Variables|Projectile|SetbStopSimulatingOnHit true)
  (Networking|SwitchHasAuthority
    (:Authority
      (bind _asfort_pawn (Utilities|Casting|CastToFortPawn _output_get)
        (:then
          (Variables|Default|SetStuckToPawn _asfort_pawn))
        (:CastFailed
          (CallFunction|HandleHits _output_get OtherComp)))
      (Utilities|FlowControl|Gate false
        (:Exit
          (Utilities|FlowControl|DoOnce false
            (:Completed
              (CallFunction|HandleTimerSelection (Variables|Default|GetTouchedActor))
              (Gameplay|AuthorityAttachtoActor (Variables|Default|GetFortAttachableProjectile) HitResult (Variables|Default|GetStuckToPawn) (Variables|Default|GetOffsetDistanceFromPhysicsMesh) (Variables|Default|GetOffsetDistanceFromBone) 0.0))))) ; Prevent atachment to pawns if already stopped.
    )))

(event OnComponentBeginOverlap(CapsuleComponent)
  (params (OverlappedComponent :type Primitive%20Component%20Object%20Reference) (OtherActor :type Actor%20Object%20Reference) (OtherComp :type Primitive%20Component%20Object%20Reference) (OtherBodyIndex :type Integer) (bFromSweep :type Boolean) (SweepResult :type Hit%20Result%20Structure%20%28by%20ref%29))
  ;[1 1 1] EVENT -> On Overlap: Larger Pawn Overlap Sphere
  (bind _output_get (Variables|Default|SetTouchedActor OtherActor))
  (if (Utilities|Equal(Object) _output_get (Game|GetInstigator))
    (if (Variables|Default|GetIsSelfStickDelayOver?)
      _))
  (Networking|SwitchHasAuthority
    (:Authority
      ;[1 1 1] EVENT -> OnTouched
      (bind _asfort_pawn (Utilities|Casting|CastToFortPawn OtherActor)
        (:then
          (Variables|Default|SetStuckToPawn _asfort_pawn)
          (Utilities|FlowControl|Gate false
            (:Exit
              (Utilities|FlowControl|DoOnce false
                (:Completed
                  (CallFunction|HandleTimerSelection (Variables|Default|GetTouchedActor))
                  (Gameplay|AuthorityAttachtoActor (Variables|Default|GetFortAttachableProjectile) (stale EventOnTouched HitResult) (Variables|Default|GetStuckToPawn) (Variables|Default|GetOffsetDistanceFromPhysicsMesh) (Variables|Default|GetOffsetDistanceFromBone) 0.0))))) ; Prevent atachment to pawns if already stopped.
        )
        (:CastFailed
          (bind _returnvalue (GameplayTags|HasAnyMatchingGameplayTags(Message) OtherActor (Variables|Default|GetTC_ActorTagsThatShouldExplodeOnOverlap)))
          (if _returnvalue
            (CallFunction|EventTriggerExplosion)))))))

(event Custom|ReopenPawnCollision
  ;[1 1 1] EVENT -> OnTouched
  ;;[1 1 1] Called when resimulating
  (Utilities|FlowControl|Sequence
    (:then_0
      (Utilities|FlowControl|Gate false
        (:Exit
          _)) ; Prevent atachment to pawns if already stopped.
    )
    (:then_1
      _)
    (:then_2
      ;[1 1 1] EVENT -> OnStop - Stop the rotation and bind to destruction of building actor if relevant.
      (Utilities|FlowControl|DoOnce false
        (:Completed
          (Gameplay|AuthorityAttachtoActor (Variables|Default|GetFortAttachableProjectile) (stale OnProjectileStop(ProjectileMovementComponent) ImpactResult) (.HitActor (Collision|BreakHitResult (stale OnProjectileStop(ProjectileMovementComponent) ImpactResult))) (Variables|Default|GetOffsetDistanceFromPhysicsMesh) (Variables|Default|GetOffsetDistanceFromBone) 0.0)))))
  (Utilities|FlowControl|DoOnce false
    (:Completed
      (CallFunction|HandleTimerSelection (Variables|Default|GetTouchedActor))
      (Gameplay|AuthorityAttachtoActor (Variables|Default|GetFortAttachableProjectile) (stale EventOnTouched HitResult) (Variables|Default|GetStuckToPawn) (Variables|Default|GetOffsetDistanceFromPhysicsMesh) (Variables|Default|GetOffsetDistanceFromBone) 0.0))))

(event OnProjectileStop(ProjectileMovementComponent)
  (params (ImpactResult :type Hit%20Result%20Structure%20%28by%20ref%29))
  (bind (_bblockinghit _binitialoverlap _time _distance _location _impactpoint _normal _impactnormal _physmat _hitactor) (Collision|BreakHitResult ImpactResult))
  (bind _returnvalue (GameplayCue|MakeGameplayCueParametersfromHitResult ImpactResult))
  (bind (_asfort_player_pawn _bsuccess) (Utilities|Casting|CastToFortPlayerPawn _hitactor))
  (bind _beep (Variables|Default|GetBeep))
  ;[1 1 1] EVENT -> OnStop - Stop the rotation and bind to destruction of building actor if relevant.
  ;;[1 0.5569 0.2257] Spawn sticky goo
  (Variables|Default|SetStickyCueParameters _returnvalue)
  (GameplayCue|AddGameplayCueOnActor(Looping) self (Variables|Default|GetStickyCueTag) _returnvalue)
  (Rendering|SetVisibility (Variables|B_Prj_Grenade|GetArrowMesh) false false)
  (Variables|Default|SetbSimulationRunning false)
  (Variables|Default|SetbHasStopped true)
  (Variables|Default|SetHitNormal _impactnormal)
  (if _bblockinghit
    (Variables|Default|SetHitLocation _location)
    (else
      (Variables|Default|SetHitLocation (Transformation|GetActorLocation))))
  (CallFunction|StopRotation)
  (Utilities|FlowControl|Sequence
    (:then_0
      ;[1 1 1] EVENT -> OnTouched
      (Utilities|FlowControl|Gate false
        (:Exit
          (Utilities|FlowControl|DoOnce false
            (:Completed
              (CallFunction|HandleTimerSelection (Variables|Default|GetTouchedActor))
              (Gameplay|AuthorityAttachtoActor (Variables|Default|GetFortAttachableProjectile) (stale EventOnTouched HitResult) (Variables|Default|GetStuckToPawn) (Variables|Default|GetOffsetDistanceFromPhysicsMesh) (Variables|Default|GetOffsetDistanceFromBone) 0.0))))) ; Prevent atachment to pawns if already stopped.
    )
    (:then_1
      (Utilities|FlowControl|Sequence
        (:then_0
          (bind _asbuilding_actor (Utilities|Casting|CastToBuildingActor _hitactor)
            (:then
              (EventDispatchers|BindEventtoOnDied _asbuilding_actor (delegate Custom|Event_BuildingActorDied))
              (Utilities|FlowControl|DoOnce false
                (:Completed
                  (Gameplay|AuthorityAttachtoActor (Variables|Default|GetFortAttachableProjectile) ImpactResult _hitactor (Variables|Default|GetOffsetDistanceFromPhysicsMesh) (Variables|Default|GetOffsetDistanceFromBone) 0.0))))
            (:CastFailed
              ;[1 1 1] Bind to when the vehicle is
              ; destroyed so we can resimulate
              (bind _asfort_athena_vehicle (Utilities|Casting|CastToFortAthenaVehicle _hitactor))
              (Game|BindEventtoOnDestroyed _asfort_athena_vehicle (delegate Custom|Event_OnVehicleDestroyed)))))
        (:then_1
          ;[1 1 1] Audio after hit
          (Networking|CanExecuteCosmeticEvents
            (:True
              (Audio|Components|Audio|Stop (Variables|B_Prj_Grenade|GetGrenadeFuse_AudioComponent))
              (CallFunction|PlayStickSound (Variables|Default|GetHitLocation) _hitactor _bsuccess)
              (Utilities|FlowControl|DoOnce false
                (:Completed
                  (Audio|Components|Audio|Play _beep 0.0)
                  (Fortnite|ShowSoundIndicatorLocalforTeam :Location (Transformation|GetActorLocation) :MaxAudibleDistance (Variables|/Script/Engine.SoundBase|GetMaxDistance (Variables|/Script/Engine.AudioComponent|GetSound _beep)) :IndicatorType "Gunshot" :Team (Team|GetActorTeam self) :Affiliations (Utilities|Array|MakeArray "Hostile" "Neutral") :Tint "(R=0.000000,G=0.000000,B=0.000000,A=0.000000)" :AttenuationMaxRangeMultiplier 1.0 :AttenuationIntensityMultiplier 1.0))))))))
    (:then_2
      (Class|BulletWhipTrackerComponentBase|Disable (Variables|Visual|GetBulletWhipTrackerComponent)))))

(event Custom|Event_BuildingActorDied
  (params (DamagedActor :type Actor%20Object%20Reference) (Damage :type Float%20%28single-precision%29) (InstigatedBy :type Controller%20Object%20Reference) (DamageCauser :type Actor%20Object%20Reference) (HitLocation :type Vector) (FHitComponent :type Primitive%20Component%20Object%20Reference) (BoneName :type Name) (Momentum :type Vector))
  ;[1 1 1] EVENT -> OnStop - Stop the rotation and bind to destruction of building actor if relevant.
  (CallFunction|RestartSimulating))

(event EventEvent_Building_Actor_Destroyed
  ;[1 1 1] EVENT -> OnStop - Stop the rotation and bind to destruction of building actor if relevant.
  (CallFunction|RestartSimulating))

(event Custom|SetBounceVelocity
  (params (Velocity :type Vector))
  ;[1 1 1] Set Bounce Velocity
  (Variables|/Script/Engine.MovementComponent|SetVelocity Velocity (Variables|Movement|GetProjectileMovementComponent)))

(event Custom|Event_OnVehicleDestroyed
  (params (DestroyedActor :type Actor%20Object%20Reference))
  ;[1 1 1] Bind to when the vehicle is
  ; destroyed so we can resimulate
  (CallFunction|RestartSimulating))

(event Custom|Stop_Rotation
  ;[1 1 1] Stop Rotating Movement
  (Components|Movement|SetUpdatedComponent (Variables|Default|GetRotatingMovement) 0)
  (Collision|SetCollisionEnabled (Variables|Collision|GetCapsuleComponent) "NoCollision")
  (Transformation|SetActorLocationAndRotation (Variables|Default|GetHitLocation) (Math|Conversions|RotationFromXVector (Math|Vector|NegateVector (Variables|Default|GetHitNormal)))))

(event EventDestroyed
  ;[1 1 1] EVENT -> Destroyed
  (CallFunction|EventTriggerExplosion))

(event Custom|OnAttachedToDied_1
  (params (DamagedActor :type Actor%20Object%20Reference) (Damage :type Float%20%28single-precision%29) (InstigatedBy :type Controller%20Object%20Reference) (DamageCauser :type Actor%20Object%20Reference) (HitLocation :type Vector) (FHitComponent :type Primitive%20Component%20Object%20Reference) (BoneName :type Name) (Momentum :type Vector))
  ;[1 1 1] EVENT -> OnAttachedDied
  (CallFunction|RestartSimulating))

(event Custom|RestartSimulating
  ;[1 1 1] Restart Simulating
  (Variables|Default|SetbSimulationRunning true)
  (Networking|ForceNetUpdate))

(event OnProjectileDetached(FortAttachableProjectile)
  (params (HostActor :type Actor%20Object%20Reference) (DetachReason :type EFortProjectileDetachReason%20Enum :default "HostDied"))
  ;[1 1 1] Handle Touched actor destroyed
  (if (Variables|Default|GetB_ExplodeOnDestroyedTouchedActor)
    (Delivery|DoExplosion)
    (else
      (Utilities|Set|Add (Variables|Default|GetIgnore) HostActor)
      (Variables|Default|SetbSimulationRunning true)
      (Networking|ForceNetUpdate)
      (Projectile|ResumeSimulation))))

(event Custom|Event_TriggerExplosion
  ;[1 1 1] Trigger Explosion
  (Utilities|FlowControl|DoOnce false
    (:Completed
      (GameplayTags|AddGameplayTag (GameplayTags|GetOwnedGameplayTags :!class /Script/GameplayTags.GameplayTagAssetInterface) "(TagName=\"Gameplay.Damage.Physical.Explosive\")" :!class /Script/GameplayTags.BlueprintGameplayTagLibrary)
      (GameplayTags|AddGameplayTag (GameplayTags|GetOwnedGameplayTags :!class /Script/GameplayTags.GameplayTagAssetInterface) "(TagName=\"Gameplay.Damage.Physical.Explosive.MoonflaxCrossbow\")" :!class /Script/GameplayTags.BlueprintGameplayTagLibrary)
      (Delivery|DoExplosion))))

(event EventOnExploded
  (params (HitActors :type Array%20of%20Actor%20Object%20References) (HitResults :type Array%20of%20Hit%20Result%20Structures))
  (bind _returnvalue (Transformation|GetActorLocation))
  (bind _cue_distantsound (Variables|Sounds|GetCue_DistantSound))
  ;[1 1 1] EVENT -> OnExploded - Emitter, Sound, Camera Shake
  (Networking|CanExecuteCosmeticEvents
    (:True
      (Variables|Default|SetIsDead? true)
      (bind _returnvalue_2 (Effects|Components|ParticleSystem|SpawnFXAtLocation (Effects|Components|ParticleSystem|GetFXSystemAsset (Variables|Default|GetNiagaraExplosionParticles)) (+ (Transformation|GetActorLocation) "0.000000,0.000000,50.000000")))
      (Rendering|SetRenderCustomDepth _returnvalue_2 true)
      (GameplayCue|ExecuteGameplayCue%28Burst%29Local self (Variables|Default|GetExplosionFeedbackCue) (Ability|GameplayCue|MakeGameplayCueParameters 0.0 1e+01)) ; Cam shake + controller vibration
      ;;[1 0.5569 0.2257] Remove sticky cue
      (GameplayCue|RemoveGameplayCueOnActor(Looping) self (Variables|Default|GetStickyCueTag) (Variables|Default|GetStickyCueParameters))
      (Components|DestroyComponent (Variables|Default|GetBeep))
      (Utilities|FlowControl|Sequence
        (:then_0
          (if (Pawn|IsLocallyControlled (Game|GetInstigator))
            (Audio|PlaySoundatLocation (Variables|Sounds|GetCue_CloseSound) _returnvalue) ; Play Close Explosion
            (else
              (Audio|PlaySoundatLocation _cue_distantsound (Transformation|GetActorLocation)) ; Play Distant Explosion
            ))
          (Fortnite|ShowSoundIndicatorLocalforTeam :Location _returnvalue :MaxAudibleDistance (Variables|/Script/Engine.SoundBase|GetMaxDistance _cue_distantsound) :IndicatorType "Gunshot" :Team (Team|GetActorTeam self) :Affiliations (Utilities|Array|MakeArray "Hostile" "Neutral" "Friendly") :Tint "(R=0.000000,G=0.000000,B=0.000000,A=0.000000)" :AttenuationMaxRangeMultiplier 1.0 :AttenuationIntensityMultiplier 1.0)
          (Audio|Components|Audio|Stop (Variables|Default|GetBeep)) ; Stop beeping sound to prevent overlap
        )
        (:then_1
          (GameplayCue|ExecuteGameplayCue%28Burst%29Local self (Variables|Default|GetExplosionFeedbackCue) (Ability|GameplayCue|MakeGameplayCueParameters 0.0)) ; Cam shake + controller vibration
          (Development|SetHiddeninGame (Variables|B_Prj_Grenade|GetArrowMesh) true false))))))

(event OnEnterWaterEvent(WaterInteractionComponent)
  (params (WaterBody :type Fort%20Water%20Body%20Actor%20Object%20Reference) (WaterInteractionComponent :type Fort%20Water%20Interaction%20Component%20Object%20Reference) (bIsFirstBody :type Boolean))
  ;[1 1 1] Handle Entering Water, Play splash FX and Sounds then trigger the beeping and the explosion event
  (Utilities|FlowControl|DoOnce false
    (:Completed
      (Default|PlaySplashFX WaterInteractionComponent)
      (Variables|/Script/Engine.ProjectileMovementComponent|SetbRotationFollowsVelocity false (Variables|Movement|GetProjectileMovementComponent))
      (Components|Movement|StopMovementImmediately (Variables|Movement|GetProjectileMovementComponent)))))

(fn DefineDelayToExplode
  (params (Time :type Scalable%20Float%20Structure))
  (Utilities|Time|SetTimerbyEvent (EventDispatchers|CreateEvent :!event Event_TriggerExplosion) (ScalableFloat|GetValueatLevel Time)) ; Time Before Explosion
)

(fn ExplosiveBowReference
  (return (Utilities|Casting|CastToB_ExplosiveBow_Athena (Variables|/Script/FortniteGame.FortPawn|GetCurrentWeapon (Utilities|Casting|CastToFortPlayerPawn (Game|GetInstigator))))))

(fn HandleHitBuildingWall
  (params (Object :type Object%20Reference) (Component :type Primitive%20Component%20Object%20Reference))
  (bind _asbuilding_wall (Utilities|Casting|CastToBuildingWall Object))
  (LocalVariables|SetLocalWall _asbuilding_wall)
  (bind _asstatic_mesh_component (Utilities|Casting|CastToStaticMeshComponent Component))
  (bind _returnvalue (BuildingWall|IsDoorComponent (LocalVariables|GetLocalWall) _asstatic_mesh_component))
  (if _returnvalue
    (Transformation|AttachActorToComponent Component "None" "KeepWorld" "KeepWorld" "KeepWorld" false)))

(fn HandleHits
  (params (Object :type Object%20Reference) (Component :type Primitive%20Component%20Object%20Reference))
  (Utilities|FlowControl|Sequence
    (:then_0
      (CallFunction|HandleHitBuildingWall Object Component))
    (:then_1
      (CallFunction|HandleHitVehicle Object Component))
    (:then_2
      (CallFunction|HandleHitSupplyDrop Object Component))))

(fn HandleHitSupplyDrop
  (params (Object :type Object%20Reference) (Component :type Primitive%20Component%20Object%20Reference))
  (Utilities|FlowControl|Sequence
    (:then_0
      (Utilities|Casting|CastToAthenaSupplyDrop Object)
      (Transformation|AttachActorToComponent Component "None" "KeepWorld" "KeepWorld" "KeepWorld"))
    (:then_1
      (Utilities|Casting|CastToAthenaSupplyDropBalloon Object)
      (Transformation|AttachActorToComponent Component "None" "KeepWorld" "KeepWorld" "KeepWorld" false))))

(fn HandleHitVehicle
  (params (Object :type Object%20Reference) (Component :type Primitive%20Component%20Object%20Reference))
  (Utilities|Casting|CastToFortAthenaVehicle Object)
  (Transformation|AttachActorToComponent Component "None" "KeepWorld" "KeepWorld" "KeepWorld" false))

(fn HandleTimerSelection
  (params (TouchedActor :type Actor%20Object%20Reference))
  (bind _returnvalue (GameplayTags|HasAnyMatchingGameplayTags(Message) TouchedActor (Variables|Default|GetVehicleTag)))
  (if _returnvalue
    (Default|DefineDelaytoExplode (Variables|Default|GetTimeBeforeExplodeVehicleHit))
    (else
      (Utilities|Casting|CastToFortPawn TouchedActor
        (:then
          (Default|DefineDelaytoExplode (Variables|Default|GetTimeBeforeExplodePlayerHit)))
        (:CastFailed
          (Default|DefineDelaytoExplode (Variables|Default|GetTimeBeforeExplode)))))))

(fn OnRep_bHasStopped)

(fn OnRep_bResumeSimulation
  (if (Variables|Default|GetbSimulationRunning)
    (Components|Movement|SetUpdatedComponent (Variables|Movement|GetProjectileMovementComponent) (Variables|Collision|GetCapsuleComponent))
    (Collision|SetCollisionEnabled (Variables|Collision|GetCapsuleComponent) "QueryOnly")
    (CallFunction|ReopenPawnCollision)
    (Networking|CanExecuteCosmeticEvents
      (:True
        (Components|Movement|SetUpdatedComponent (Variables|Default|GetRotatingMovement) (Variables|B_Prj_Grenade|GetArrowMesh))))))

(fn OnRep_HitLocation)

(fn OnRep_HitNormal)

(fn PlaySplashFX
  (params (WaterInteractionComponent :type Fort%20Water%20Interaction%20Component%20Object%20Reference))
  (bind (_waterplanelocation _waterplanenormal _watersurfaceposition _waterdepth _waterbodyidx) (FX|GetLastWaterSurfaceInfo WaterInteractionComponent :!class /Script/FortniteGame.FortWaterInteractionComponent))
  (Variables|Default|SetHitLocation (Transformation|GetActorLocation))
  (Variables|Default|SetHitNormal "0, 0,1.000000")
  (bind _returnvalue (Effects|SpawnEmitteratLocation (Variables|Default|GetWaterSplashFX) _watersurfaceposition (Math|Rotator|MakeRotfromX _waterplanenormal)))
  (Effects|Components|ParticleSystem|SetFloatParameter _returnvalue "WaterDepth" _waterdepth)
  (Effects|Components|ParticleSystem|SetFloatParameter _returnvalue "WaterBodyIndex" (Math|Conversions|ToFloat(Integer) _waterbodyidx))
  (Audio|PlaySoundatLocation "/Game/Sounds/Fort_GamePlay_Sounds/Explosives/Grenade_Projectile_Bounce_Water_Splash_NoLoop_Cue.Grenade_Projectile_Bounce_Water_Splash_NoLoop_Cue" (Transformation|GetActorLocation)))

(fn PlayStickSound
  (params (Location :type Vector) (HitActor :type Actor%20Object%20Reference) (Success :type Boolean))
  (bind _returnvalue (Game|GetInstigator))
  (if Success
    (Utilities|IsValid _returnvalue
      (:Is%20Valid
        (if (Pawn|IsLocallyControlled _returnvalue)
          (Audio|PlaySound2D (Variables|Default|GetAudio_StuckToPlayer)) ; Plays on player who threw grenade
          (else
            (Utilities|IsValid HitActor
              (:Is%20Valid
                (if (Pawn|IsLocallyControlled (Utilities|Casting|CastToFortPlayerPawn HitActor))
                  (Audio|PlaySound2D (Variables|Default|GetAudio_StuckToPlayer)) ; Plays on player hit with grenade
                  (else
                    _)))
              (:Is%20Not%20Valid
                _)))))
      (:Is%20Not%20Valid
        _))
    (Audio|PlaySoundatLocation (Variables|Default|GetAudio_StuckToEnvironment) Location)
    (else
      (Audio|PlaySoundatLocation (Variables|Default|GetAudio_StuckToEnvironment) Location))))

(fn SetGravityAndSpeed)

(fn ConstructionScript
  (Utilities|FlowControl|Sequence
    (:then_0
      (Networking|CanExecuteCosmeticEvents
        (:True
          (Components|Movement|SetUpdatedComponent (Variables|Default|GetRotatingMovement) (Variables|B_Prj_Grenade|GetArrowMesh)))))
    (:then_1
      (if (== (ScalableFloat|GetValueatLevel (Variables|Default|GetF_ExplodeOnDestroyedTouchedActor)) 0)
        (Variables|Default|SetB_ExplodeOnDestroyedTouchedActor false)
        (else
          (Variables|Default|SetB_ExplodeOnDestroyedTouchedActor true))))))
