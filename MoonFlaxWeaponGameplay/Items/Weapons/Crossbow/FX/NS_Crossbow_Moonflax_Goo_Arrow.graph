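# Niagara effect for the Moonflax crossbow's goo arrow, built from a single mesh-particle emitter named Goo.
# Spawns small, dark, randomly scaled blobs at a steady rate in a flattened shape around the system location, nudged along X.
# Each blob lives about five seconds, swelling and shrinking back to zero on a scale curve.
# Dynamic material parameters and per-particle material randomness feed variation to the mesh material.
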NiagaraSystem /MoonFlaxWeaponGameplay/Items/Weapons/Crossbow/FX/NS_Crossbow_Moonflax_Goo_Arrow
dependencies {usedDataInterfaces [&NiagaraDataInterfaceCurve], usedDynamicInputs [], usedModules [&SystemState]}
emitters {Goo {
  emitterUpdateScript [
   &EmitterState()
   &SpawnRate(Use Spawn Probability=false)
  ]
  particleSpawnScript [
   &InitializeParticle(
    Color=<rgba 0.026 0.0237 0.0234 1>
    Lifetime Max=5
    Lifetime Min=4.5
    Lifetime Mode=::NewEnumerator1
    Mesh Scale=&Multiply_VectorByFloat(Vector=&RandomRangeVector(Minimum=<0.06 0.05 0.05>, Maximum=<0.08 0.1 0.1>), Float=0.35)
    Mesh Scale Mode=::NewEnumerator4
    Position Offset=&RandomRangeVector(Minimum=<0 0 0>, Maximum=<2 0 0>)
    UsePositionOffset=true
    Write Material Random=true
   )
   &ShapeLocation(MaskSpawnGroup=false, Non Uniform Scale=<1 1 0.1>, Override Seed=false, Rotation Mode=::NewEnumerator1, Sphere Radius=1, UseOwnerScale=true)
   &InitialMeshOrientation(Mesh Orientation Mode=::NewEnumerator2, Rotation Coordinate Space=::NewEnumerator2, Use Rotation Vector=true) {module=InitialMeshOrientation001}
   &SystemLocation(MaskSpawnGroup=false, Offset=&RandomRangeVector(Minimum=<4 0 0>, Maximum=<3.5 0 0>))
  ]
  particleUpdateScript [
   &ParticleState()
   &ScaleMeshSize(Scale Factor=&VectorFromFloat(Value=&FloatFromCurve(FloatCurve=DataInterface() {propertyValues {Curve {keys [
          {arriveTangent=7.0823, interpMode=RCIM_Cubic, leaveTangent=7.0823, tangentMode=RCTM_User, time=0, value=0}
          {interpMode=RCIM_Cubic, tangentMode=RCTM_User, time=0.5, value=1}
          {arriveTangent=-6.4477, tangentMode=RCTM_User, time=1, value=0}
         ]}}})))
   &DynamicMaterialParameters(
    Index 0 Param 1=0.03
    Index 0 Param 2=&RandomRangeFloat(Minimum=-0.1, Maximum=0.1)
    Module.Param0WriteEnabled=true
    Module.Param1WriteEnabled=true
    Module.Param2WriteEnabled=false
    Module.Param3WriteEnabled=false
   )
   &SolveForcesAndVelocity(Clamp Velocity=false, Limit Acceleration=false)
  ]
  rendererClasses [&NiagaraMeshRendererProperties]
  renderers [{rendererClass=&NiagaraMeshRendererProperties, rendererIndex=0}]
 }}
summary {
 systemName=NS_Crossbow_Moonflax_Goo_Arrow
 emitters [{emitterName=Goo, rendererClasses [&NiagaraMeshRendererProperties]}]
 userVariables []
}
