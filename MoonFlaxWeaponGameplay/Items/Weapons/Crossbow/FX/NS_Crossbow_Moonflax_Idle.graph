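# Niagara system for the Moonflax crossbow's idle flame effect, rendered as looping mesh particles.
# A burst spawns two flames, each following a bone or socket on the skeletal mesh by partition index.
# One component of the owner's velocity drives mesh rotation rate.
# While User.isShooting is true, red is boosted and system Timeline curves dip alpha to zero and back while swelling mesh size before settling.
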NiagaraSystem /MoonFlaxWeaponGameplay/Items/Weapons/Crossbow/FX/NS_Crossbow_Moonflax_Idle
dependencies {
 usedDataInterfaces [&NiagaraDataInterfaceCurve, &NiagaraDataInterfaceSkeletalMesh]
 usedDynamicInputs []
 usedModules [
  &self:SystemSpawnScript.SystemScriptSource.SystemScriptGraph.NiagaraNodeAssignment_1.SetVariables_#1
  &SystemState
  &Timeline
 ]
}
emitters {Flames_Idle {
  emitterUpdateScript [
   &EmitterState(
    Inactive Response=::NewEnumerator1
    Life Cycle Mode=::NewEnumerator1
    Loop Behavior=::NewEnumerator1
    Loop Duration Mode=::NewEnumerator1
    UseLoopDelay=false
   )
   &SpawnBurst_Instantaneous(Spawn Count=2, Use Spawn Probability=false, UseLoopCountLimit=false)
  ]
  particleSpawnScript [
   &PartitionParticles()
   &InitializeParticle(
    Lifetime Max=0.5
    Lifetime Min=0.1
    Lifetime Mode=::NewEnumerator1
    Mesh Scale Mode=::NewEnumerator3
    Mesh Uniform Scale=1.3
    UsePositionOffset=false
    Write Material Random=true
   )
   &InitialMeshOrientation(Mesh Orientation Mode=::NewEnumerator7, Rotation=<-0.1 0 0>, Use Rotation Vector=true)
  ]
  particleUpdateScript [
   &ParticleState(
    Kill Particles When Lifetime Has Elapsed=false
    Let Infinitely Lived Particles Die When Emitter Deactivates=true
    Loop Particles Lifetime=true
   )
   &SkeletalMeshLocation(
    Bone / Socket Index=$Particles.PartitionParticles.Partition
    MaskSpawnGroup=false
    Mesh Sampling Type=::NewEnumerator5
    Skeletal Mesh=$System.Skeletal Mesh
    Socket Sampling Mode=::NewEnumerator5
   )
   &DynamicMaterialParameters(
    Index 0 Param 1=4
    Index 1 Param 1=58.9487
    Module.Index0Param0Write=true
    Module.Index0Param1Write=true
    Module.Index0Param2Write=true
    Module.Index0Param3Write=true
    Module.Index1Param0Write=true
    Module.Index1Param1Write=true
    Module.Index1Param2Write=true
    Module.Index1Param3Write=true
    Write Parameter Index 1=true
   )
   &UpdateMeshOrientation(
    Rotation Rate=&Multiply_Float(A=&MakeFloatFromVector(Vector=$Engine.Owner.Velocity, Channel=::NewEnumerator0), B=0.01)
    Rotation Vector=<0 1 0>
   )
   &ScaleColor(
    Scale Alpha=&MakeCustomFloatFromBool(
     Bool=$User.isShooting
     True Float=&FloatFromCurve(
      FloatCurve=DataInterface() {propertyValues {Curve {keys [
          {arriveTangent=-6.6044, interpMode=RCIM_Cubic, leaveTangent=-6.6044, tangentMode=RCTM_User, time=0, value=1}
          {arriveTangent=-1.0631, interpMode=RCIM_Cubic, leaveTangent=-1.0631, tangentMode=RCTM_User, time=0.2, value=0}
          {interpMode=RCIM_Cubic, tangentMode=RCTM_User, time=0.5, value=0}
          {arriveTangent=0.5128, interpMode=RCIM_Cubic, leaveTangent=0.5128, tangentMode=RCTM_User, time=0.8, value=0}
          {arriveTangent=13.6446, tangentMode=RCTM_User, time=1, value=1}
         ]}}}
      CurveIndex=$System.Timeline.NormalizedTime
     )
     False Float=1
    )
    Scale RGB=&MakeVector(X=&MakeCustomFloatFromBool(Bool=$User.isShooting, True Float=4, False Float=1), Y=1, Z=1)
    ScaleA=true
    ScaleRGB=true
   )
   &ScaleMeshSize(Scale Factor=&VectorFromFloat(Value=&MakeCustomFloatFromBool(
      Bool=$User.isShooting
      True Float=&FloatFromCurve(
       FloatCurve=DataInterface() {propertyValues {Curve {keys [
           {interpMode=RCIM_Cubic, leaveTangent=-6.832, tangentMode=RCTM_Break, time=0, value=1}
           {arriveTangent=8.4842, interpMode=RCIM_Cubic, leaveTangent=1, tangentMode=RCTM_Break, time=0.6, value=2}
           {arriveTangent=-2.4817, interpMode=RCIM_Cubic, leaveTangent=-2.4817, tangentMode=RCTM_User, time=1, value=1}
          ]}}}
       CurveIndex=$System.Timeline.NormalizedTime
      )
      False Float=1
     ))) {module=ScaleMeshSize003}
   &CameraOffset(Camera Offset Amount=1)
  ]
  rendererClasses [&NiagaraMeshRendererProperties]
  renderers [{rendererClass=&NiagaraMeshRendererProperties, rendererIndex=0}]
 }}
summary {
 systemName=NS_Crossbow_Moonflax_Idle
 emitters [{emitterName=Flames_Idle, rendererClasses [&NiagaraMeshRendererProperties]}]
 userVariables [User.isShooting {type=NiagaraBool}]
}
