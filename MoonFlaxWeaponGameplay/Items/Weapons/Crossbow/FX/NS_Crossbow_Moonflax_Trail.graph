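# Niagara smoke trail for the Moonflax crossbow, drawn as a single ribbon emitter.
# Particles spawn at a steady rate, with position blending from User.MuzzlePosition to the simulation position over the first moments of system age.
# Curl noise and drag make the ribbon drift as it widens.
# Color flashes HDR orange, then turns white and fades, with alpha also scaled by emitter age and an NPC time-of-day parameter.
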NiagaraSystem /MoonFlaxWeaponGameplay/Items/Weapons/Crossbow/FX/NS_Crossbow_Moonflax_Trail
dependencies {
 usedDataInterfaces [&NiagaraDataInterfaceCurve, &NiagaraDataInterfaceColorCurve]
 usedDynamicInputs []
 usedModules [&SystemState]
}
emitters {Smoke_Trail {
  emitterUpdateScript [
   &EmitterState()
   &SpawnRate(SpawnRate=60, Use Spawn Probability=false)
  ]
  particleSpawnScript [&InitializeParticle(
    Lifetime=2
    Position=&Lerp_Vector(
     A=$User.MuzzlePosition
     B=/Niagara/DynamicInputs/Helpers/SimulationPosition()
     Alpha=&FloatFromCurve(
      FloatCurve=DataInterface() {propertyValues {Curve {keys [
          {arriveTangent=9.109, interpMode=RCIM_Cubic, leaveTangent=9.109, tangentMode=RCTM_User, time=0.02, value=0}
          {interpMode=RCIM_Cubic, leaveTangent=2.5007, time=0.2, value=1}
         ]}}}
      CurveIndex=$Engine.System.Age
     )
    )
    Position Mode=::NewEnumerator1
    Ribbon Width=13
    Ribbon Width Mode=::NewEnumerator1
    UsePositionOffset=false
    Write Material Random=true
   )]
  particleUpdateScript [
   &ParticleState()
   &CurlNoiseForce(Noise Frequency=10, Noise Strength=50, Pan Noise=true, Pan Noise Field=<0 0 1>)
   &Drag(Drag=1.5, Use Linear Drag=true, Use Rotational Drag=true)
   &SolveForcesAndVelocity(Clamp Velocity=false, Limit Acceleration=false)
   &ScaleRibbonWidth(Ribbon Width Scale=&Multiply_Float(
     A=&FloatFromCurve(FloatCurve=DataInterface() {propertyValues {Curve {keys [
          {arriveTangent=0.0296, interpMode=RCIM_Cubic, leaveTangent=0.0296, tangentMode=RCTM_User, time=0, value=1}
          {interpMode=RCIM_Cubic, leaveTangent=-0.9375, time=1, value=2}
         ]}}})
     B=1
    ))
   &FN_RibbonMaterial(
    Color=&ColorFromCurve(ColorCurve=DataInterface() {propertyValues {
       AlphaCurve {keys [<at 0 1 RCIM_Cubic>,<at 1 0 RCIM_Cubic>]}
       BlueCurve {keys [<at 0 2.1047 RCIM_Cubic>,<at 0.0702 1>,<at 1 1 RCIM_Cubic>]}
       GreenCurve {keys [<at 0 10.2546 RCIM_Cubic>,<at 0.0702 1>,<at 1 1 RCIM_Cubic>]}
       RedCurve {keys [<at 0 33 RCIM_Cubic>,<at 0.0702 1>,<at 1 1 RCIM_Cubic>]}
      }})
    Scale Alpha:NiagaraFloat=&Multiply_Float(
     A=&FloatFromCurve(
      FloatCurve=DataInterface() {propertyValues {Curve {keys [
          {interpMode=RCIM_Cubic, tangentMode=RCTM_User, time=0, value=1}
          {arriveTangent=-1.4505, interpMode=RCIM_Cubic, leaveTangent=-1.4505, tangentMode=RCTM_User, time=1, value=0}
         ]}}}
      CurveIndex=$Emitter.Age
     )
     B=&FloatFromCurve(
      FloatCurve=DataInterface() {propertyValues {Curve {keys [
          {interpMode=RCIM_Cubic, tangentMode=RCTM_User, time=0, value=0.4}
          {interpMode=RCIM_Cubic, tangentMode=RCTM_User, time=0.5, value=1}
          {tangentMode=RCTM_User, time=1, value=0.4}
         ]}}}
      CurveIndex=Linked() {linkedVariable {name=NPC.FortniteNPC.TimeOfDay_0-1, type=NiagaraFloat}}
     )
    )
   ) {module=FN_RibbonMaterial001}
  ]
  rendererClasses [&NiagaraRibbonRendererProperties]
  renderers [{rendererClass=&NiagaraRibbonRendererProperties}]
 }}
summary {
 systemName=NS_Crossbow_Moonflax_Trail
 emitters [{emitterName=Smoke_Trail, rendererClasses [&NiagaraRibbonRendererProperties]}]
 userVariables [User.MuzzlePosition {type=Vector3f, description="", defaultValue {struct=Vector3f, value=<0 0 0>}}]
}
