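# Sound cue for the Moonflax crossbow's aim-down-sights entry foley.
# A mixer layers a pitch-modulated random pick of two ADS-in one-shots with a quieter random pick of three looping ADS waves.
# Because the loop layer is set to loop, the cue keeps sounding after the entry one-shot ends.
# It plays in the weapon foley sound class with reload-style attenuation at a reduced overall volume.
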SoundCue /MoonFlaxWeaponGameplay/Items/Weapons/Crossbow/Sounds/Moonflax_Crossbow_ADS_In_Cue
root=Mixer_0
nodes [
 {id=Mixer_0, children [Modulator_0,Random_1], properties {inputVolume [1,0.25]}}
 {id=Modulator_0, children [Random_0], properties {pitchMax=1.2}}
 {id=Random_0, children [WavePlayer_2,WavePlayer_4], properties {weights [1,1]}}
 {id=Random_1, children [WavePlayer_6,WavePlayer_7,WavePlayer_8], properties {weights [1,1,1]}}
 {id=WavePlayer_2, properties {soundWaveAssetPtr=/MoonFlaxWeaponGameplay/Items/Weapons/Crossbow/Sounds/wavs/Moonflax_Crossbow_ADS_In_01}}
 {id=WavePlayer_4, properties {soundWaveAssetPtr=/MoonFlaxWeaponGameplay/Items/Weapons/Crossbow/Sounds/wavs/Moonflax_Crossbow_ADS_In_02}}
 {
  id=WavePlayer_6
  properties {bLooping=true, soundWaveAssetPtr=/MoonFlaxWeaponGameplay/Items/Weapons/Crossbow/Sounds/wavs/Moonflax_Crossbow_ADS_LP_01}
 }
 {
  id=WavePlayer_7
  properties {bLooping=true, soundWaveAssetPtr=/MoonFlaxWeaponGameplay/Items/Weapons/Crossbow/Sounds/wavs/Moonflax_Crossbow_ADS_LP_02}
 }
 {
  id=WavePlayer_8
  properties {bLooping=true, soundWaveAssetPtr=/MoonFlaxWeaponGameplay/Items/Weapons/Crossbow/Sounds/wavs/Moonflax_Crossbow_ADS_LP_03}
 }
]
properties {
 attenuationSettings=&Fort_Weapon_Reload_Att
 maxDistance=2250
 soundClassObject=&Fort_WeaponFoley
 volumeMultiplier=0.35
}
