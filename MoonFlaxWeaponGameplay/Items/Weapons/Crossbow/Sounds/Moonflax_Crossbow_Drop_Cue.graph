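# Sound cue for the Moonflax crossbow drop sound.
# A Random node picks among three equally weighted drop waves, and a root Modulator adds slight volume variation.
# The cue plays at reduced volume through the shared weapon drop attenuation and Fort_Loot_Collect sound class, with the Weapon_Drop_StopOldest concurrency set managing overlapping drops.
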SoundCue /MoonFlaxWeaponGameplay/Items/Weapons/Crossbow/Sounds/Moonflax_Crossbow_Drop_Cue
root=Modulator_1
nodes [
 {id=Modulator_1, children [Random_0], properties {volumeMax=1.02, volumeMin=0.98}}
 {id=Random_0, children [WavePlayer_0,WavePlayer_1,WavePlayer_2], properties {weights [1,1,1]}}
 {id=WavePlayer_0, properties {soundWaveAssetPtr=/MoonFlaxWeaponGameplay/Items/Weapons/Crossbow/Sounds/wavs/Moonflax_Crossbow_Drop_01}}
 {id=WavePlayer_1, properties {soundWaveAssetPtr=/MoonFlaxWeaponGameplay/Items/Weapons/Crossbow/Sounds/wavs/Moonflax_Crossbow_Drop_02}}
 {id=WavePlayer_2, properties {soundWaveAssetPtr=/MoonFlaxWeaponGameplay/Items/Weapons/Crossbow/Sounds/wavs/Moonflax_Crossbow_Drop_03}}
]
properties {
 attenuationSettings=&Fort_Weapon_Drop_Att
 maxDistance=3000
 soundClassObject=&Fort_Loot_Collect
 volumeMultiplier=0.3
 concurrencyOverrides {resolutionRule=PreventNew}
 concurrencySet [&Weapon_Drop_StopOldest]
}
