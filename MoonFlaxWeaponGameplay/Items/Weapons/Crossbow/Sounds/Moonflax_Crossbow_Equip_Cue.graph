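# Sound cue for equipping the Moonflax crossbow, played quietly in the weapon foley sound class.
# A random node picks one of three equally weighted equip waves, then a modulator varies pitch and volume.
# An IsLocal branch routes that shared chain through an attenuation node without assigned settings when true, or through weapon reload attenuation when false or unset.
# The cue overrides concurrency and has its debug flag on.
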SoundCue /MoonFlaxWeaponGameplay/Items/Weapons/Crossbow/Sounds/Moonflax_Crossbow_Equip_Cue
root=Branch_0
nodes [
 {id=Attenuation_0, children [Modulator_0]}
 {id=Attenuation_1, children [Modulator_0], properties {attenuationSettings=&Fort_Weapon_Reload_Att}}
 {id=Branch_0, children [Attenuation_0, Attenuation_1, Attenuation_1], properties {boolParameterName=IsLocal}}
 {id=Modulator_0, children [Random_1], properties {pitchMax=1, volumeMax=1}}
 {id=Random_1, children [WavePlayer_2,WavePlayer_3,WavePlayer_4], properties {weights [1,1,1]}}
 {id=WavePlayer_2, properties {soundWaveAssetPtr=/MoonFlaxWeaponGameplay/Items/Weapons/Crossbow/Sounds/wavs/Moonflax_Crossbow_Equip_01}}
 {id=WavePlayer_3, properties {soundWaveAssetPtr=/MoonFlaxWeaponGameplay/Items/Weapons/Crossbow/Sounds/wavs/Moonflax_Crossbow_Equip_02}}
 {id=WavePlayer_4, properties {soundWaveAssetPtr=/MoonFlaxWeaponGameplay/Items/Weapons/Crossbow/Sounds/wavs/Moonflax_Crossbow_Equip_03}}
]
properties {
 bDebug=true
 bOverrideConcurrency=true
 maxDistance=2097152
 soundClassObject=&Fort_WeaponFoley
 volumeMultiplier=0.3
}
