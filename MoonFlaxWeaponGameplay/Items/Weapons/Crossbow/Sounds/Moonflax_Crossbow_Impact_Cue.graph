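# Sound cue for the Moonflax crossbow's drop impact, playing one of three drop impact waves.
# An equally weighted random node picks the wave and feeds a modulator with default settings at the root.
# The cue plays at reduced volume through the Fort_Loot_Collect sound class, with overridden short-range attenuation.
# Its concurrency override caps simultaneous instances at a few and refuses new ones beyond that.
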SoundCue /MoonFlaxWeaponGameplay/Items/Weapons/Crossbow/Sounds/Moonflax_Crossbow_Impact_Cue
root=Modulator_0
nodes [
 {id=Modulator_0, children [Random_0]}
 {id=Random_0, children [WavePlayer_0,WavePlayer_1,WavePlayer_2], properties {weights [1,1,1]}}
 {id=WavePlayer_0, properties {soundWaveAssetPtr=/MoonFlaxWeaponGameplay/Items/Weapons/Crossbow/Sounds/wavs/Moonflax_Crossbow_Drop_Impact_01}}
 {id=WavePlayer_1, properties {soundWaveAssetPtr=/MoonFlaxWeaponGameplay/Items/Weapons/Crossbow/Sounds/wavs/Moonflax_Crossbow_Drop_Impact_02}}
 {id=WavePlayer_2, properties {soundWaveAssetPtr=/MoonFlaxWeaponGameplay/Items/Weapons/Crossbow/Sounds/wavs/Moonflax_Crossbow_Drop_Impact_03}}
]
properties {
 bOverrideAttenuation=true
 bOverrideConcurrency=true
 maxDistance=1000
 soundClassObject=&Fort_Loot_Collect
 volumeMultiplier=0.2
 attenuationOverrides {falloffDistance=1000, attenuationShapeExtents {x=0}}
 concurrencyOverrides {maxCount=3, resolutionRule=PreventNew}
}
