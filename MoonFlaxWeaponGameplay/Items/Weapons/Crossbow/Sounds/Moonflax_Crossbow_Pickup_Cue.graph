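# Sound cue for picking up the Moonflax crossbow, routed to the loot-collect sound class.
# A Random node picks one of three equally weighted pickup waves, and a Modulator adds very slight volume variation.
# Playback is quiet, uses the shared weapon-pickup attenuation and is limited by the PickupMaximum_Oldest concurrency set.
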SoundCue /MoonFlaxWeaponGameplay/Items/Weapons/Crossbow/Sounds/Moonflax_Crossbow_Pickup_Cue
root=Modulator_0
nodes [
 {id=Modulator_0, children [Random_0], properties {volumeMax=1.01, volumeMin=0.99}}
 {id=Random_0, children [WavePlayer_1,WavePlayer_2,WavePlayer_3], properties {weights [1,1,1]}}
 {id=WavePlayer_1, properties {soundWaveAssetPtr=/MoonFlaxWeaponGameplay/Items/Weapons/Crossbow/Sounds/wavs/Moonflax_Crossbow_Pickup_01}}
 {id=WavePlayer_2, properties {soundWaveAssetPtr=/MoonFlaxWeaponGameplay/Items/Weapons/Crossbow/Sounds/wavs/Moonflax_Crossbow_Pickup_02}}
 {id=WavePlayer_3, properties {soundWaveAssetPtr=/MoonFlaxWeaponGameplay/Items/Weapons/Crossbow/Sounds/wavs/Moonflax_Crossbow_Pickup_03}}
]
properties {
 attenuationSettings=&Fort_Weapon_Pickup_Att
 maxDistance=3500
 soundClassObject=&Fort_Loot_Collect
 volumeMultiplier=0.25
 concurrencyOverrides {maxCount=1, resolutionRule=PreventNew}
 concurrencySet [&PickupMaximum_Oldest]
}
