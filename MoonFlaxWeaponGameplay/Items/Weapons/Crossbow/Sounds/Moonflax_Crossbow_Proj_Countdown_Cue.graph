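# Sound cue for the Moonflax crossbow projectile's countdown, built on a landing timer sound.
# A single wave player feeds the output directly, with no randomization or modulation.
# It uses the shared explosive-beep attenuation and gadget-items sound class, plays at reduced volume, and overrides concurrency to cap simultaneous instances.
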SoundCue /MoonFlaxWeaponGameplay/Items/Weapons/Crossbow/Sounds/Moonflax_Crossbow_Proj_Countdown_Cue
root=WavePlayer_0
nodes [{
  id=WavePlayer_0
  properties {soundWaveAssetPtr=/MoonFlaxWeaponGameplay/Items/Weapons/Crossbow/Sounds/wavs/Moonflax_Crossbow_Proj_Land_Timer_01}
 }]
properties {
 attenuationSettings=/Game/Sounds/Attenuation/Fort_ExplosiveBeep_Medium_Att
 bOverrideConcurrency=true
 maxDistance=6500
 soundClassObject=&Fort_Gadgets_Items
 volumeMultiplier=0.5
 concurrencyOverrides {maxCount=3}
}
