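# Sound cue for the pump sound in the MoonFlax crossbow reload.
# A Random node picks among six equally weighted pump wave variations, and a Modulator adds slight pitch and volume variation.
# It plays at reduced volume in the weapon foley sound class, with weapon reload attenuation and overridden concurrency.
# The debug flag is left enabled on the cue.
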SoundCue /MoonFlaxWeaponGameplay/Items/Weapons/Crossbow/Sounds/Moonflax_Crossbow_Reload_Pump_Cue
root=Modulator_0
nodes [
 {id=Modulator_0, children [Random_1], properties {pitchMax=1.02, pitchMin=0.99, volumeMax=1.02, volumeMin=0.98}}
 {id=Random_1, children [WavePlayer_0,WavePlayer_4,WavePlayer_5,WavePlayer_6,WavePlayer_7,WavePlayer_8], properties {weights [1,1,1,1,1,1]}}
 {id=WavePlayer_0, properties {soundWaveAssetPtr=/MoonFlaxWeaponGameplay/Items/Weapons/Crossbow/Sounds/wavs/Moonflax_Crossbow_Reload_Pump_01}}
 {id=WavePlayer_4, properties {soundWaveAssetPtr=/MoonFlaxWeaponGameplay/Items/Weapons/Crossbow/Sounds/wavs/Moonflax_Crossbow_Reload_Pump_02}}
 {id=WavePlayer_5, properties {soundWaveAssetPtr=/MoonFlaxWeaponGameplay/Items/Weapons/Crossbow/Sounds/wavs/Moonflax_Crossbow_Reload_Pump_03}}
 {id=WavePlayer_6, properties {soundWaveAssetPtr=/MoonFlaxWeaponGameplay/Items/Weapons/Crossbow/Sounds/wavs/Moonflax_Crossbow_Reload_Pump_04}}
 {id=WavePlayer_7, properties {soundWaveAssetPtr=/MoonFlaxWeaponGameplay/Items/Weapons/Crossbow/Sounds/wavs/Moonflax_Crossbow_Reload_Pump_05}}
 {id=WavePlayer_8, properties {soundWaveAssetPtr=/MoonFlaxWeaponGameplay/Items/Weapons/Crossbow/Sounds/wavs/Moonflax_Crossbow_Reload_Pump_06}}
]
properties {
 attenuationSettings=&Fort_Weapon_Reload_Att
 bDebug=true
 bOverrideConcurrency=true
 maxDistance=2250
 soundClassObject=&Fort_WeaponFoley
 volumeMultiplier=0.4
}
